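; SleepyTime drop for PinkOatmeal: a launched projectile pickup that gains gravity on first hit and stops where it lands.
; Interacting applies the consume gameplay effect with a puff, while destruction of its attached object spawns loot pickups server-side.
; A failsafe timer from BeginPlay despawns it unopened; the post-landing Delay leads nowhere.
; Hide and Landed RepNotifies drive visibility, collision, trail and ambient audio, alongside replicated gameplay cues.
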(event Collision|EventHit
  (params (MyComp :type Primitive%20Component%20Object%20Reference) (Other :type Actor%20Object%20Reference) (OtherComp :type Primitive%20Component%20Object%20Reference) (bSelfMoved :type Boolean) (HitLocation :type Vector :default "0, 0, 0") (HitNormal :type Vector :default "0, 0, 0") (NormalImpulse :type Vector :default "0, 0, 0") (Hit :type Hit%20Result%20Structure%20%28by%20ref%29))
  ;[1 1 1] Turn On gravity on first hit
  (|CallParentFunction MyComp Other OtherComp bSelfMoved HitLocation HitNormal NormalImpulse Hit)
  (Utilities|FlowControl|DoOnce false
    (:Completed
      (Variables|/Script/Engine.ProjectileMovementComponent|SetProjectileGravityScale (Variables|Default|GetProjectileScaleAfterBounce) (Variables|Default|GetProjectileMovement)))))

(event Custom|FailsafeKill
  ;[1 1 1] Failsafe Delay
  (CallFunction|Despawn))

(event EventBeginPlay
  (bind _projectilemovement (Variables|Default|GetProjectileMovement))
  ;[1 1 1] Play Spawn GC and set speed
  (Ability|ExecuteGameplayCueReplicated (Variables|Default|GetGC_Spawn))
  (bind _output_get (Variables|/Script/Engine.ProjectileMovementComponent|SetInitialSpeed (ScalableFloat|GetValueatLevel (Variables|Default|GetRow_Speed)) _projectilemovement))
  (bind _output_get_2 (Variables|/Script/Engine.ProjectileMovementComponent|SetMaxSpeed _output_get _projectilemovement))
  (Variables|/Script/Engine.MovementComponent|SetVelocity (* (Transformation|GetActorForwardVector) _output_get_2) _projectilemovement)
  (Utilities|Time|SetTimerbyEvent (delegate Custom|FailsafeKill) (+ (ScalableFloat|GetValueatLevel (Variables|Default|GetRow_LifeAfterLand)) (Variables|Default|GetFailsafeAddedLife)))
  (|CallParentFunction))

(event Collision|EventActorBeginOverlap
  (params (OtherActor :type Actor%20Object%20Reference)))

(event Custom|Despawn
  ;[1 1 1] Despawn
  (if (not (Variables|Default|GetOpened))
    (Variables|Default|SetHide true)
    (Ability|ExecuteGameplayCueReplicated (Variables|Default|GetGC_Despawn))
    (Actor|SetLifeSpan 0.2)))

(event EventTick
  (params (DeltaSeconds :type Float%20%28single-precision%29)))

(event EventStopSim
  (params (Hit :type Hit%20Result%20Structure))
  ;[1 1 1] On Stop Sim - Set grav to 0, Play Landed FX and set lifetime after land. Despawn if not opened before lifetime ends.
  (|CallParentFunction Hit)
  (Variables|/Script/Engine.ProjectileMovementComponent|SetProjectileGravityScale 0.0 (Variables|Default|GetProjectileMovement))
  (Ability|ExecuteGameplayCueReplicated (Variables|Default|GetGC_Land))
  (if (Networking|IsServer)
    (Variables|Default|SetLanded true)
    (Utilities|FlowControl|Delay (ScalableFloat|GetValueatLevel (Variables|Default|GetRow_LifeAfterLand)))))

(event Interaction|EventBlueprintOnInteract
  (params (InteractingPawn :type Fort%20Pawn%20Object%20Reference) (InteractionBeingAttempted :type EInteractionBeingAttempted%20Enum :default "FirstInteraction"))
  ;[1 1 1] Handle Open - If the object it landed on was destroyed or it was interacted with then open the actor. On open hide, play the Open GC, Spawn Items, and set lifetime.
  (GameplayEffects|ApplyGameplayEffectToTarget :self (Ability|GetAbilitySystemComponent self) :GameplayEffectClass "/Figment_S10/Athena/Items/EnvironmentalItems/PinkOatmeal_10_4/SleepyTime/GE_PinkOatmeal_Consume.GE_PinkOatmeal_Consume_C" :Target (Ability|GetAbilitySystemComponent InteractingPawn) :Level 0.0)
  (Variables|Default|SetHide true)
  (Variables|Default|SetOpened true)
  (CallFunction|PuffPickup)
  (Actor|SetLifeSpan 2.0))

(event Custom|PuffPickup
  ;[1 1 1] Handle Open - If the object it landed on was destroyed or it was interacted with then open the actor. On open hide, play the Open GC, Spawn Items, and set lifetime.
  (Effects|SpawnEmitteratLocation (Variables|Default|GetFX_Pickup) (Transformation|GetActorLocation) (Transformation|GetActorRotation)))

(event EventAttachedWasDestroyed
  ;[1 1 1] Handle Open - If the object it landed on was destroyed or it was interacted with then open the actor. On open hide, play the Open GC, Spawn Items, and set lifetime.
  (CallFunction|OnOpen)
  (|CallParentFunction))

(event Custom|OnOpen
  ;[1 1 1] Handle Open - If the object it landed on was destroyed or it was interacted with then open the actor. On open hide, play the Open GC, Spawn Items, and set lifetime.
  (Variables|Default|SetOpened true)
  (CallFunction|SpawnItems (Transformation|GetActorLocation) (Transformation|GetActorForwardVector))
  (Variables|Default|SetHide true)
  (Ability|ExecuteGameplayCueReplicated (Variables|Default|GetGC_Open))
  (Actor|SetLifeSpan 1.0))

(event Custom|SpawnItems
  (params (SpawnLocation :type Vector) (Direction :type Vector))
  ;[1 1 1] SpawnItems
  (if (Networking|IsServer)
    (Development|PrintString "Spawn Items")
    (bind (_outloottodrop _returnvalue) (Loot|PickLootDrops (Variables|Default|GetItemsToDrop) 1 -1 "(GameplayTags=)"))
    (if _returnvalue
      (for _array_element _outloottodrop
        (bind (_array_element_itemdefinition _array_element_itemguid _array_element_count) (Class|FortItemFunctionLibrary|BreakFortItemEntry _array_element))
        (Development|PrintString (Utilities|String|Append "Item To Spawn:" (Utilities|GetDisplayName _array_element_itemdefinition)))
        (bind _asfort_world_item_definition (Utilities|Casting|CastToFortWorldItemDefinition _array_element_itemdefinition))
        (Inventory|SpawnPickupInWorld _asfort_world_item_definition _array_element_count (+ SpawnLocation (Variables|Default|GetSpawnPositionOffset)) Direction 0 true false false -1 "Container" "Drone")))))

(fn BlueprintCanInteract
  (params (InteractingPawn :type Fort%20Pawn%20Object%20Reference) (InteractionBeingAttempted :type EInteractionBeingAttempted%20Enum :default "FirstInteraction") (InteractionType :type TInteractionType%20Enum :default "IT_NoInteraction"))
  (return true))

(fn BlueprintGetInteractionString
  (params (InteractingPawn :type Fort%20Pawn%20Object%20Reference) (InteractionBeingAttempted :type EInteractionBeingAttempted%20Enum :default "FirstInteraction"))
  (return (Variables|Default|GetInteractText)))

(fn OnRep_Hide
  (bind _hide (Variables|Default|GetHide))
  (Rendering|SetActorHiddenInGame _hide)
  (Collision|SetActorEnableCollision (not _hide))
  (if (Variables|Default|GetHide)
    (Audio|Components|Audio|FadeOut (Variables|Default|GetAmbient) 1.0 0.0 "Linear" :!class /Script/Engine.AudioComponent)))

(fn OnRep_Landed
  (bind _trail (Variables|Default|GetTrail))
  (bind _ambient (Variables|Default|GetAmbient))
  (if (Variables|Default|GetLanded)
    (Components|Activation|Deactivate _trail)
    (Audio|Components|Audio|SetSound _ambient (Variables|Default|GetAmbientLanded))
    (else
      (Components|Activation|Activate _trail false)
      (Audio|Components|Audio|SetSound _ambient (Variables|Default|GetAmbientInAir)))))

(fn ShouldDie
  (params (Damage :type Float%20%28single-precision%29) (EventInstigator :type Controller%20Object%20Reference) (DamageCauser :type Actor%20Object%20Reference) (bIsTest :type Boolean) (EffectContext :type Gameplay%20Effect%20Context%20Handle%20Structure%20%28by%20ref%29))
  (return false))

(fn ConstructionScript
  (|CallParentFunction))
